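# Post-process AnimBlueprint for the KoboLobo male cosmetic head, adjusting its pose and tongue differently in the front end and elsewhere.
# Outside the front end, a Control Rig receives parachute, skydive-float and emote weights, then spring-damped AnimDynamics simulates the tongue with overridden gravity.
# In the front end, the rig receives only the emote weight, and fixed additive bone offsets pose the tongue instead of physics.
# The event graph smooths the state booleans into those weights with FInterpTo.
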AnimBlueprint /BRCosmetics/Characters/Player/Male/Medium/Heads/M_MED_KoboLobo_Head/Meshes/M_MED_KoboLobo_Head_PostProcessBP
target_skeleton=/BRCosmetics/Characters/Player/Male/Medium/Heads/M_MED_KoboLobo_Head/Meshes/CustomSkeleton/M_MED_KoboLobo_Head_Skeleton
graphs [
 AnimGraph {nodes [
   Output Pose(Result=Blend Poses by bool(BlendPose_0=Use cached pose 'frontend pose'(), BlendPose_1=Use cached pose 'main pose'(), bActiveValue=Get bIsInFrontEnd()) {blendType=Linear})
   Save cached pose 'start pose'(Pose=Input Pose()) {cacheName=start pose}
   n1 = Control Rig!(Source=Use cached pose 'start pose'(), Alpha, Is Playing Emote:double=Get Is Playing Emote() -> (Is Playing Emote:double)) -> (Pose) {
    alphaBoolBlend {blendOption=Linear}
    controlRigAssetReference {blueprintRigClass=&/M_MED_KoboLobo_Head_CtrlRig.M_MED_KoboLobo_Head_CtrlRig_C}
   }
   n2 = "Transform (Modify) Bone - Bone: tongue"(ComponentPose=Local To Component(LocalPose=@n1), Alpha=1, Translation="1.5,-2.5,-11", Rotation="2.5,4.3,-30", Scale="1,1,1") {
    rotationMode=BMM_Additive
    rotationSpace=BCS_BoneSpace
    translationMode=BMM_Additive
    translationSpace=BCS_BoneSpace
    alphaBoolBlend {blendOption=Linear}
    boneToModify {boneName=tongue}
    rotation=<rot 2.5 4.3 -30>
    translation=<1.5 -2.5 -11>
   }
   n3 = "Transform (Modify) Bone - Bone: dyn_tongue_01"(ComponentPose=@n2, Alpha=1, Translation="0,0,0", Rotation="-15,-1,3", Scale="1,1,1") {
    rotationMode=BMM_Additive
    rotationSpace=BCS_BoneSpace
    translationMode=BMM_Additive
    translationSpace=BCS_BoneSpace
    alphaBoolBlend {blendOption=Linear}
    boneToModify {boneName=dyn_tongue_01}
    rotation=<rot -15 -1 3>
   }
   Save cached pose 'frontend pose'(Pose=Component To Local(ComponentPose=@n3)) {cacheName=frontend pose}
   n4 = Control Rig!(
    Source=Use cached pose 'start pose'()
    Alpha
    Is Parachute Opened:double=Get Is Parachute Opened() -> (Is Parachute Opened:double)
    Is Skydive Floating:double=Get Is Skydive Floating() -> (Is Skydive Floating:double)
    Is Playing Emote=Get Is Playing Emote()
   ) -> (Pose) {
    alphaBoolBlend {blendOption=Linear}
    controlRigAssetReference {blueprintRigClass=&M_MED_KoboLobo_Head_CtrlRig_C}
   }
   n5 = Anim Dynamics - Bone: dyn_tongue_01(ComponentPose=Local To Component(LocalPose=@n4), Alpha=1, ExternalForce="0,0,0") {
    angularDampingOverride=0.999
    angularSpringConstant=2000
    bAngularSpring=true
    bGravityOverrideInSimSpace=true
    bOverrideAngularDamping=true
    bUseGravityOverride=true
    simulationSpace=BoneRelative
    alphaBoolBlend {blendOption=Linear}
    boundBone {boneName=dyn_tongue_01}
    componentAppliedLinearAccClamp=<5000 5000 5000>
    componentLinearAccScale=<1 1 1>
    componentLinearVelScale=<1 1 1>
    gravityOverride {x=-150}
    physicsBodyDefinitions [{
      boundBone {boneName=dyn_tongue_01}
      boxExtents=<15 13 13>
      constraintSetup {angularLimitsMax=<0 5 15>, angularLimitsMin=<0 -10 -15>, angularTarget=<0 0 0>, linearAxesMin=<-0.5 0 0>}
      localJointOffset=<6 0 0>
     }]
    relativeSpaceBone {boneName=pelvis}
   }
   Save cached pose 'main pose'(Pose=Component To Local(ComponentPose=@n5)) {cacheName=main pose}
  ]}
 EventGraph {nodes [
   n1 = Event BlueprintUpdateAnimation()
   > Set Delta Time X(Delta Time X=@n1.DeltaTimeX)
   n2 = Get bIsParachuteOpened()
   > Set Is Parachute Opened(Is Parachute Opened:double=FInterpTo(Current=Get Is Parachute Opened(), Target="To Float (Boolean)"(InBool=@n2), DeltaTime=Get Delta Time X(), InterpSpeed=10)) -> (Output_Get:double)
   n3 = Get bIsSkydiveFloating()
   > Set Is Skydive Floating(Is Skydive Floating:double=FInterpTo(Current=Get Is Skydive Floating(), Target="To Float (Boolean)"(InBool=@n3), DeltaTime=Get Delta Time X(), InterpSpeed=10)) -> (Output_Get:double)
   n4 = Get bIsPlayingEmote()
   > Set Is Playing Emote(Is Playing Emote:double=FInterpTo(Current=Get Is Playing Emote(), Target="To Float (Boolean)"(InBool=@n4), DeltaTime=Get Delta Time X(), InterpSpeed=10)) -> (Output_Get:double)
  ]}
]
variables [
 Delta Time X {type=number}
 Is Parachute Opened {type=number}
 Is Playing Emote {type=number}
 Is Skydive Floating {type=number}
]
